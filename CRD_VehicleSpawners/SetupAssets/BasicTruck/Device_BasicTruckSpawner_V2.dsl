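; Basic Truck vehicle spawner device (V2) that exposes a Color and Style user option for the truck.
; ColorAndStyleAsAnInteger returns that enum option as an integer minus one, and SetColorOption reverses the offset to write an index back as EColorBasicTruck.
; The construction script only calls its parent, and the event graph holds placeholder events with no active logic.
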(event EventBeginPlay
  (|CallParentFunction) ; This node is disabled and will not be called.\nDrag off pins to build functionality.
)

(event Collision|EventActorBeginOverlap
  (params (OtherActor :type Actor%20Object%20Reference)))

(event EventTick
  (params (DeltaSeconds :type Float%20%28single-precision%29)))

(fn ColorAndStyleAsAnInteger
  (return (- (Math|Conversions|ToInteger(Byte) (Variables|UserOptions|Basic|GetColor%20and%20Style)) 1)))

(fn SetColorOption
  (params (ColorIndex :type Integer))
  (Variables|UserOptions|Basic|SetColor%20and%20Style (Utilities|Enum|BytetoEnumEColorBasicTruck (Math|Conversions|ToByte(Integer) (+ ColorIndex 1)))))

(fn ConstructionScript
  (|CallParentFunction))
